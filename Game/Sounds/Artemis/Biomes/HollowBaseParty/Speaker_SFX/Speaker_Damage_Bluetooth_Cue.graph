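# Sound cue for damage to a Bluetooth speaker in the HollowBaseParty biome, routed through impact class and SFX submix with large loudspeaker attenuation.
# A mixer layers a random small speaker impact with a random connection crackle that follows after a short random delay at reduced volume.
# A top modulator varies overall pitch and volume, and the cue doubles pitch.
# Overridden concurrency limits copies per owner, blocking new ones past the cap.
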SoundCue /Game/Sounds/Artemis/Biomes/HollowBaseParty/Speaker_SFX/Speaker_Damage_Bluetooth_Cue
root=Modulator_2
nodes [
 {id=Delay_0, children [Modulator_0], properties {delayMax=0.2, delayMin=0.1}}
 {id=Mixer_0, children [Modulator_1,Delay_0], properties {inputVolume [1,1]}}
 {id=Modulator_0, children [Random_0], properties {volumeMax=0.8, volumeMin=0.5}}
 {id=Modulator_1, children [Random_2]}
 {id=Modulator_2, children [Mixer_0], properties {pitchMax=1.1, pitchMin=0.9, volumeMin=0.7}}
 {id=Random_0, children [WavePlayer_0,WavePlayer_1,WavePlayer_2,WavePlayer_3], properties {weights [1,1,1,1]}}
 {id=Random_2, children [WavePlayer_9, WavePlayer_10, WavePlayer_11, WavePlayer_12, WavePlayer_13], properties {weights [1,1,1,1,1]}}
 {id=WavePlayer_0, properties {soundWaveAssetPtr=/Game/Sounds/Artemis/Biomes/HollowBaseParty/Speaker_SFX/wav/Speaker_ConnectionCrackle_01}}
 {id=WavePlayer_1, properties {soundWaveAssetPtr=/Game/Sounds/Artemis/Biomes/HollowBaseParty/Speaker_SFX/wav/Speaker_ConnectionCrackle_02}}
 {id=WavePlayer_10, properties {soundWaveAssetPtr=/Game/Sounds/Artemis/Biomes/HollowBaseParty/Speaker_SFX/wav/Speaker_Sm_Impact_02}}
 {id=WavePlayer_11, properties {soundWaveAssetPtr=/Game/Sounds/Artemis/Biomes/HollowBaseParty/Speaker_SFX/wav/Speaker_Sm_Impact_03}}
 {id=WavePlayer_12, properties {soundWaveAssetPtr=/Game/Sounds/Artemis/Biomes/HollowBaseParty/Speaker_SFX/wav/Speaker_Sm_Impact_04}}
 {id=WavePlayer_13, properties {soundWaveAssetPtr=/Game/Sounds/Artemis/Biomes/HollowBaseParty/Speaker_SFX/wav/Speaker_Sm_Impact_05}}
 {id=WavePlayer_2, properties {soundWaveAssetPtr=/Game/Sounds/Artemis/Biomes/HollowBaseParty/Speaker_SFX/wav/Speaker_ConnectionCrackle_03}}
 {id=WavePlayer_3, properties {soundWaveAssetPtr=/Game/Sounds/Artemis/Biomes/HollowBaseParty/Speaker_SFX/wav/Speaker_ConnectionCrackle_04}}
 {id=WavePlayer_9, properties {soundWaveAssetPtr=/Game/Sounds/Artemis/Biomes/HollowBaseParty/Speaker_SFX/wav/Speaker_Sm_Impact_01}}
]
properties {
 attenuationSettings=&/Loudspeaker_Damage_Large
 bOverrideConcurrency=true
 maxDistance=9150
 pitchMultiplier=2
 soundClassObject=&Fort_Impacts
 soundSubmixObject=&Fortnite_SFX
 volumeMultiplier=1
 attenuationOverrides {falloffDistance=4650, attenuationShapeExtents {x=800}}
 concurrencyOverrides {bLimitToOwner=true, maxCount=2, resolutionRule=PreventNew, retriggerTime=1.2, volumeScale=0.3}
}
