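# Sound cue for damage to a large speaker in the HollowBaseParty biome, layering an impact hit with a trailing electrical crackle.
# A Mixer plays a random large impact immediately and a random connection crackle after a short randomized delay, each with slight volume modulation.
# It uses large-loudspeaker attenuation, the impacts sound class, and per-owner concurrency that limits overlapping plays and blocks new ones beyond the cap.
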SoundCue /Game/Sounds/Artemis/Biomes/HollowBaseParty/Speaker_SFX/Speaker_Damage_Lrg_Cue
root=Mixer_0
nodes [
 {id=Delay_0, children [Modulator_0], properties {delayMax=0.2, delayMin=0.18}}
 {id=Mixer_0, children [Modulator_1,Delay_0], properties {inputVolume [1,1]}}
 {id=Modulator_0, children [Random_0], properties {volumeMax=1, volumeMin=0.7}}
 {id=Modulator_1, children [Random_1], properties {volumeMax=0.8, volumeMin=0.7}}
 {id=Random_0, children [WavePlayer_0,WavePlayer_1,WavePlayer_2,WavePlayer_3], properties {weights [1,1,1,1]}}
 {id=Random_1, children [WavePlayer_4,WavePlayer_5,WavePlayer_6,WavePlayer_7,WavePlayer_8], properties {weights [1,1,1,1,1]}}
 {id=WavePlayer_0, properties {soundWaveAssetPtr=/Game/Sounds/Artemis/Biomes/HollowBaseParty/Speaker_SFX/wav/Speaker_ConnectionCrackle_01}}
 {id=WavePlayer_1, properties {soundWaveAssetPtr=/Game/Sounds/Artemis/Biomes/HollowBaseParty/Speaker_SFX/wav/Speaker_ConnectionCrackle_02}}
 {id=WavePlayer_2, properties {soundWaveAssetPtr=/Game/Sounds/Artemis/Biomes/HollowBaseParty/Speaker_SFX/wav/Speaker_ConnectionCrackle_03}}
 {id=WavePlayer_3, properties {soundWaveAssetPtr=/Game/Sounds/Artemis/Biomes/HollowBaseParty/Speaker_SFX/wav/Speaker_ConnectionCrackle_04}}
 {id=WavePlayer_4, properties {soundWaveAssetPtr=/Game/Sounds/Artemis/Biomes/HollowBaseParty/Speaker_SFX/wav/Speaker_Lrg_Impact_01}}
 {id=WavePlayer_5, properties {soundWaveAssetPtr=/Game/Sounds/Artemis/Biomes/HollowBaseParty/Speaker_SFX/wav/Speaker_Lrg_Impact_02}}
 {id=WavePlayer_6, properties {soundWaveAssetPtr=/Game/Sounds/Artemis/Biomes/HollowBaseParty/Speaker_SFX/wav/Speaker_Lrg_Impact_03}}
 {id=WavePlayer_7, properties {soundWaveAssetPtr=/Game/Sounds/Artemis/Biomes/HollowBaseParty/Speaker_SFX/wav/Speaker_Lrg_Impact_04}}
 {id=WavePlayer_8, properties {soundWaveAssetPtr=/Game/Sounds/Artemis/Biomes/HollowBaseParty/Speaker_SFX/wav/Speaker_Lrg_Impact_05}}
]
properties {
 attenuationSettings=&/Loudspeaker_Damage_Large
 bOverrideConcurrency=true
 maxDistance=9150
 soundClassObject=&Fort_Impacts
 soundSubmixObject=&Fortnite_SFX
 volumeMultiplier=1.8
 attenuationOverrides {falloffDistance=4650, attenuationShapeExtents {x=800}}
 concurrencyOverrides {
  bLimitToOwner=true
  bVolumeScaleCanRelease=true
  maxCount=2
  resolutionRule=PreventNew
  retriggerTime=1.2
  voiceStealReleaseTime=0.4
  volumeScale=0.3
 }
}
