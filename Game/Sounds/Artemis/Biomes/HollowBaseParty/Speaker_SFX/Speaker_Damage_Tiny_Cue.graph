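# Sound cue for damage impacts on tiny speakers in the Hollow Base Party biome.
# A Random node picks one of five equally weighted tiny impact waves.
# A Modulator at the root then varies pitch and volume slightly for each play.
# The cue uses a small loudspeaker damage attenuation, the impacts sound class and a boosted volume multiplier.
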SoundCue /Game/Sounds/Artemis/Biomes/HollowBaseParty/Speaker_SFX/Speaker_Damage_Tiny_Cue
root=Modulator_2
nodes [
 {id=Modulator_2, children [Random_0], properties {pitchMax=1.1, pitchMin=0.9, volumeMin=0.7}}
 {id=Random_0, children [WavePlayer_0,WavePlayer_1,WavePlayer_2,WavePlayer_3,WavePlayer_4], properties {weights [1,1,1,1,1]}}
 {id=WavePlayer_0, properties {soundWaveAssetPtr=/Game/Sounds/Artemis/Biomes/HollowBaseParty/Speaker_SFX/wav/Speaker_Tiny_Impact_05}}
 {id=WavePlayer_1, properties {soundWaveAssetPtr=/Game/Sounds/Artemis/Biomes/HollowBaseParty/Speaker_SFX/wav/Speaker_Tiny_Impact_04}}
 {id=WavePlayer_2, properties {soundWaveAssetPtr=/Game/Sounds/Artemis/Biomes/HollowBaseParty/Speaker_SFX/wav/Speaker_Tiny_Impact_03}}
 {id=WavePlayer_3, properties {soundWaveAssetPtr=/Game/Sounds/Artemis/Biomes/HollowBaseParty/Speaker_SFX/wav/Speaker_Tiny_Impact_02}}
 {id=WavePlayer_4, properties {soundWaveAssetPtr=/Game/Sounds/Artemis/Biomes/HollowBaseParty/Speaker_SFX/wav/Speaker_Tiny_Impact_01}}
]
properties {
 attenuationSettings=&/Loudspeaker_Damage_Small
 maxDistance=4000
 soundClassObject=&Fort_Impacts
 volumeMultiplier=3
}
