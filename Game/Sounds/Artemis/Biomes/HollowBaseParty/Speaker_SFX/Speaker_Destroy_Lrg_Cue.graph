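# Sound cue for the destruction of a large speaker in the HollowBaseParty biome, classed under Fort_Impacts.
# It picks one of two death waves at equal weight and passes it through a Modulator for slight pitch and volume variation.
# Inline attenuation overrides add distance low-pass filtering, occlusion and listener focus over a long falloff.
# Concurrency overrides allow a single instance, blocking new plays and limiting retriggers to prevent stacking.
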SoundCue /Game/Sounds/Artemis/Biomes/HollowBaseParty/Speaker_SFX/Speaker_Destroy_Lrg_Cue
root=Modulator_1
nodes [
 {id=Modulator_1, children [Random_0], properties {pitchMin=0.945, volumeMax=1, volumeMin=0.9}}
 {id=Random_0, children [WavePlayer_0,WavePlayer_1], properties {weights [1,1]}}
 {id=WavePlayer_0, properties {soundWaveAssetPtr=/Game/Sounds/Artemis/Biomes/HollowBaseParty/Speaker_SFX/wav/Speaker_Lrg_Death_01}}
 {id=WavePlayer_1, properties {soundWaveAssetPtr=/Game/Sounds/Artemis/Biomes/HollowBaseParty/Speaker_SFX/wav/Speaker_Lrg_Death_02}}
]
properties {
 attenuationSettings=&/Loudspeaker_Damage_Large
 bOverrideAttenuation=true
 bOverrideConcurrency=true
 maxDistance=10500
 soundClassObject=&Fort_Impacts
 volumeMultiplier=1.6
 attenuationOverrides {
  bAttenuateWithLPF=true
  bEnableListenerFocus=true
  bEnableOcclusion=true
  falloffDistance=9000
  lPFFrequencyAtMax=900
  lPFFrequencyAtMin=8500
  lPFRadiusMax=3500
  lPFRadiusMin=850
  nonFocusVolumeAttenuation=0.8
  nonSpatializedRadiusStart=30
  occlusionLowPassFilterFrequency=2000
  occlusionTraceChannel=ECC_GameTraceChannel11
  occlusionVolumeAttenuation=0.6
  stereoSpread=3000
  attenuationShapeExtents {x=1500}
 }
 concurrencyOverrides {bVolumeScaleCanRelease=true, maxCount=1, resolutionRule=PreventNew, retriggerTime=1.2, voiceStealReleaseTime=0.3, volumeScale=0.4}
}
