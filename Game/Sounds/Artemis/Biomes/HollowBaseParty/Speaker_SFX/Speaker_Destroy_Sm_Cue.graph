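# Sound cue for the destruction of a small speaker in the Artemis HollowBaseParty biome.
# A Random node picks one of two equally weighted death waves, then a Modulator raises its pitch within a narrow randomized range.
# It uses the Fort_Impacts sound class and Loudspeaker_Damage_Small attenuation.
# Overridden concurrency allows one voice at a time, with a retrigger time, volume scaling and a voice-steal release.
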SoundCue /Game/Sounds/Artemis/Biomes/HollowBaseParty/Speaker_SFX/Speaker_Destroy_Sm_Cue
root=Modulator_1
nodes [
 {id=Modulator_1, children [Random_1], properties {pitchMax=1.3, pitchMin=1.25, volumeMax=1, volumeMin=1}}
 {id=Random_1, children [WavePlayer_2,WavePlayer_3], properties {bShouldExcludeFromBranchCulling=true, weights [1,1]}}
 {id=WavePlayer_2, properties {soundWaveAssetPtr=/Game/Sounds/Artemis/Biomes/HollowBaseParty/Speaker_SFX/wav/Speaker_Sm_Death_01}}
 {id=WavePlayer_3, properties {soundWaveAssetPtr=/Game/Sounds/Artemis/Biomes/HollowBaseParty/Speaker_SFX/wav/Speaker_Sm_Death_02}}
]
properties {
 attenuationSettings=&/Loudspeaker_Damage_Small
 bOverrideConcurrency=true
 maxDistance=4000
 soundClassObject=&Fort_Impacts
 volumeMultiplier=1
 attenuationOverrides {bAttenuateWithLPF=true, falloffDistance=4650, attenuationShapeExtents {x=800}}
 concurrencyOverrides {bVolumeScaleCanRelease=true, maxCount=1, retriggerTime=0.5, voiceStealReleaseTime=0.3, volumeScale=0.5}
}
