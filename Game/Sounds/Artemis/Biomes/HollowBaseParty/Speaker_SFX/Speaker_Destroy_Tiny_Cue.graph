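# Sound cue for destroying a tiny speaker in the HollowBaseParty biome.
# A random node picks one of three equally weighted destroy waves and passes it through a modulator for variation.
# It plays in the impacts sound class with small loudspeaker-damage attenuation and a boosted volume multiplier.
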SoundCue /Game/Sounds/Artemis/Biomes/HollowBaseParty/Speaker_SFX/Speaker_Destroy_Tiny_Cue
root=Modulator_0
nodes [
 {id=Modulator_0, children [Random_0]}
 {id=Random_0, children [WavePlayer_0,WavePlayer_1,WavePlayer_2], properties {weights [1,1,1]}}
 {id=WavePlayer_0, properties {soundWaveAssetPtr=/Game/Sounds/Artemis/Biomes/HollowBaseParty/Speaker_SFX/wav/Speaker_Tiny_Destroy_01}}
 {id=WavePlayer_1, properties {soundWaveAssetPtr=/Game/Sounds/Artemis/Biomes/HollowBaseParty/Speaker_SFX/wav/Speaker_Tiny_Destroy_02}}
 {id=WavePlayer_2, properties {soundWaveAssetPtr=/Game/Sounds/Artemis/Biomes/HollowBaseParty/Speaker_SFX/wav/Speaker_Tiny_Destroy_03}}
]
properties {
 attenuationSettings=&/Loudspeaker_Damage_Small
 maxDistance=4000
 soundClassObject=&Fort_Impacts
 volumeMultiplier=3
}
