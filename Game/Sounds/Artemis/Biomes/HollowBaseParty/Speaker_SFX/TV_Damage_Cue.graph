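# Sound cue for TV damage impacts in the HollowBaseParty biome, played as an impacts-class sound through the SFX submix.
# A Random node picks among ten equally weighted TV impact waves, and a Modulator varies pitch and volume on each play.
# It reuses the TV wall music attenuation asset for spatialization.
# Its own concurrency override tightly caps overlapping instances and sets a retrigger time.
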SoundCue /Game/Sounds/Artemis/Biomes/HollowBaseParty/Speaker_SFX/TV_Damage_Cue
root=Modulator_1
nodes [
 {id=Modulator_1, children [Random_0], properties {pitchMin=0.9, volumeMin=0.8}}
 {
  id=Random_0
  children [
   WavePlayer_0
   WavePlayer_1
   WavePlayer_2
   WavePlayer_3
   WavePlayer_9
   WavePlayer_10
   WavePlayer_11
   WavePlayer_12
   WavePlayer_13
   WavePlayer_14
  ]
  properties {weights [1,1,1,1,1,1,1,1,1,1]}
 }
 {id=WavePlayer_0, properties {soundWaveAssetPtr=/Game/Sounds/Artemis/Biomes/HollowBaseParty/Speaker_SFX/wav/TV_Impact_10}}
 {id=WavePlayer_1, properties {soundWaveAssetPtr=/Game/Sounds/Artemis/Biomes/HollowBaseParty/Speaker_SFX/wav/TV_Impact_01}}
 {id=WavePlayer_10, properties {soundWaveAssetPtr=/Game/Sounds/Artemis/Biomes/HollowBaseParty/Speaker_SFX/wav/TV_Impact_05}}
 {id=WavePlayer_11, properties {soundWaveAssetPtr=/Game/Sounds/Artemis/Biomes/HollowBaseParty/Speaker_SFX/wav/TV_Impact_06}}
 {id=WavePlayer_12, properties {soundWaveAssetPtr=/Game/Sounds/Artemis/Biomes/HollowBaseParty/Speaker_SFX/wav/TV_Impact_07}}
 {id=WavePlayer_13, properties {soundWaveAssetPtr=/Game/Sounds/Artemis/Biomes/HollowBaseParty/Speaker_SFX/wav/TV_Impact_08}}
 {id=WavePlayer_14, properties {soundWaveAssetPtr=/Game/Sounds/Artemis/Biomes/HollowBaseParty/Speaker_SFX/wav/TV_Impact_09}}
 {id=WavePlayer_2, properties {soundWaveAssetPtr=/Game/Sounds/Artemis/Biomes/HollowBaseParty/Speaker_SFX/wav/TV_Impact_02}}
 {id=WavePlayer_3, properties {soundWaveAssetPtr=/Game/Sounds/Artemis/Biomes/HollowBaseParty/Speaker_SFX/wav/TV_Impact_03}}
 {id=WavePlayer_9, properties {soundWaveAssetPtr=/Game/Sounds/Artemis/Biomes/HollowBaseParty/Speaker_SFX/wav/TV_Impact_04}}
]
properties {
 attenuationSettings=/Game/Sounds/Artemis/Biomes/HollowBaseParty/TV_Walls/HollowBaseParty_Music_TV_Wall_Att
 bOverrideConcurrency=true
 maxDistance=1350
 soundClassObject=&Fort_Impacts
 soundSubmixObject=&Fortnite_SFX
 volumeMultiplier=0.8
 attenuationOverrides {
  bEnableOcclusion=true
  falloffDistance=2400
  occlusionLowPassFilterFrequency=1000
  occlusionTraceChannel=ECC_GameTraceChannel6
  occlusionVolumeAttenuation=0.6
  attenuationShapeExtents {x=800}
 }
 concurrencyOverrides {maxCount=2, retriggerTime=0.5, volumeScale=0.3}
}
