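# Sound cue for a TV being destroyed in the HollowBaseParty biome, routed to the Fort_Impacts sound class.
# A Random node picks between two equally weighted TV death waves and feeds a Modulator at the root.
# Spatialization references the TV wall music attenuation asset.
# Overridden concurrency allows two instances, blocks new ones beyond that, and sets volume scaling and a retrigger time.
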SoundCue /Game/Sounds/Artemis/Biomes/HollowBaseParty/Speaker_SFX/TV_Destroy_Cue
root=Modulator_1
nodes [
 {id=Modulator_1, children [Random_1]}
 {id=Random_1, children [WavePlayer_0,WavePlayer_1], properties {weights [1,1]}}
 {id=WavePlayer_0, properties {soundWaveAssetPtr=/Game/Sounds/Artemis/Biomes/HollowBaseParty/Speaker_SFX/wav/TV_Death_01}}
 {id=WavePlayer_1, properties {soundWaveAssetPtr=/Game/Sounds/Artemis/Biomes/HollowBaseParty/Speaker_SFX/wav/TV_Death_02}}
]
properties {
 attenuationSettings=/Game/Sounds/Artemis/Biomes/HollowBaseParty/TV_Walls/HollowBaseParty_Music_TV_Wall_Att
 bOverrideConcurrency=true
 maxDistance=1350
 soundClassObject=&Fort_Impacts
 volumeMultiplier=1.2
 attenuationOverrides {
  bAttenuateWithLPF=true
  bEnableOcclusion=true
  falloffDistance=3200
  occlusionLowPassFilterFrequency=900
  occlusionTraceChannel=ECC_GameTraceChannel6
  occlusionVolumeAttenuation=0.6
  attenuationShapeExtents {x=800}
 }
 concurrencyOverrides {bVolumeScaleCanRelease=true, maxCount=2, resolutionRule=PreventNew, retriggerTime=1, voiceStealReleaseTime=0.3, volumeScale=0.5}
}
